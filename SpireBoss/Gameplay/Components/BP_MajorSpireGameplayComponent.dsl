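; BP_MajorSpireGameplayComponent is a Blueprint gameplay component, named for a major spire in the SpireBoss gameplay content.
; Its graph holds only a BeginPlay event with no connected logic, so it adds no behavior of its own here.
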(event EventBeginPlay)
